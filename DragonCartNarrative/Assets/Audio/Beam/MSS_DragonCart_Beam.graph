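# Mono looping MetaSound source for the DragonCart beam, playing an endless, varied bed from random wave assets with attenuation.
# Two wave players alternate, each picking a random wave and semitone pitch, with ADSR envelopes crossfading them into the mixer.
# Each handoff is delayed by pitch-corrected wave duration minus crossfade time, then loops back through deferred variables.
# Attack time switches to the crossfade time only at the first handoff, leaving the initial start unfaded.
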MetaSoundSource /DragonCartNarrative/Assets/Audio/Beam/MSS_DragonCart_Beam
inputs [
 Crossfade Curve {type=Float}
 "Crossfade Time (Seconds)" {type=Float}
 "Pitch (Semitones)" {type=Float}
 "Pitch Randomization (Semitones)" {type=Float}
 UE.Source.OnPlay {type=Trigger}
 "Volume (dB)" {type=Float}
 Wave Assets {type=WaveAsset:Array}
]
interfaces [UE.OutputFormat.Mono:1.0, UE.Source:1.1]
outputs ["UE.OutputFormat.Mono.Audio:0" {type=Audio}]
pages [Default {
  input_defaults {
   Crossfade Curve=1
   "Crossfade Time (Seconds)"=0.7
   "Pitch (Semitones)"=0
   "Pitch Randomization (Semitones)"=1
   UE.Source.OnPlay=null
   "Volume (dB)"=0
   Wave Assets [
    &/DragonCart_Beam_04
    &/DragonCart_Beam_03
    &/DragonCart_Beam_02
    &/DragonCart_Beam_01
    &/DragonCart_Beam_05
   ]
  }
  nodes [
   DecibelstoLinearGain_0 = UE.Decibels to Linear Gain:Float(Decibels=^"Volume (dB)")
   n1 = VariableDeferredAccessor.Time(Variable=InitVariable.Time().Variable)
   n2 = VariableDeferredAccessor.Trigger(Variable=InitVariable.Trigger().Variable)
   n3 = VariableDeferredAccessor.Trigger(Variable=InitVariable.Trigger().Variable)
   TriggerAnyTriggerAny2_0 = "TriggerAny.Trigger Any (2)"(In 0=^UE.Source.OnPlay, In 1=@n3.Value)
   n4 = UE.RandomFloat(
    Max=UE.Add:Float(PrimaryOperand=^"Pitch (Semitones)", AdditionalOperands=^"Pitch Randomization (Semitones)").Out
    Min=UE.Subtract:Float(PrimaryOperand=^"Pitch (Semitones)", AdditionalOperands=^"Pitch Randomization (Semitones)").Out
    Next=@TriggerAnyTriggerAny2_0.Out
   )
   ArrayRandomGet_0 = Array.Random Get:WaveAsset:Array(In Array=^Wave Assets, Next=@n4.On Next) {also {defaults {Enable Shared State="<MetaSoundBuilderResult.SUCCEEDED: 0>"}}}
   n5 = VariableAccessor.Float(Variable=VariableMutator.Float(Variable=InitVariable.Float().Variable, Value=@n4.Value).Variable)
   WavePlayer_1 = UE.Wave Player:Mono(Play=@ArrayRandomGet_0.On Next, Wave Asset=@ArrayRandomGet_0.Value, Pitch Shift=VariableAccessor.Float(Variable=@n5.Variable).Value) {also {version [1,1]}}
   ADSREnvelopeADSREnvelope_1 = ADSR Envelope.ADSR Envelope:Float(
    Attack Curve=^Crossfade Curve
    Release Curve=^Crossfade Curve
    Release Time=UE.ConversionFloatToTime(In=^"Crossfade Time (Seconds)").Out
    Attack Time=@n1.Value
    Trigger Release=@n2.Value
    Trigger Attack=@WavePlayer_1.On Play
   ) {also {defaults {Hard Reset="<MetaSoundBuilderResult.SUCCEEDED: 0>", Sustain Level=1}}}
   n6 = InitVariable.Trigger()
   n7 = VariableDeferredAccessor.Time(Variable=@n1.Variable)
   n8 = InitVariable.Trigger()
   GetWaveDuration_0 = "UE.Get Wave Duration: "(Wave=@ArrayRandomGet_0.Value)
   n9 = UE.Divide:Float(
    PrimaryOperand=UE.ConversionTimeToFloat(In=@GetWaveDuration_0.Duration).Out
    AdditionalOperands=UE.Power:Float(AdditionalOperands=UE.Divide:Float(PrimaryOperand=@n5.Value) {also {defaults {AdditionalOperands=12}}}.Out) {also {defaults {PrimaryOperand=2}}}.Out
   )
   n10 = InitVariable.Float()
   n11 = UE.Subtract:Float(
    AdditionalOperands=^"Crossfade Time (Seconds)"
    PrimaryOperand=VariableAccessor.Float(Variable=VariableMutator.Float(Value=@n9.Out, Variable=@n10.Variable).Variable).Value
   )
   TriggerDelay_1 = UE.Trigger Delay(
    In=@ADSREnvelopeADSREnvelope_1.On Attack Triggered
    Delay Time=UE.ConversionFloatToTime(In=Clamp.Clamp:Float(In=@n11.Out) {also {defaults {Max=60, Min=0.01}}}.Value).Out
   )
   n12 = UE.RandomFloat(
    Max=UE.Add:Float(PrimaryOperand=^"Pitch (Semitones)", AdditionalOperands=^"Pitch Randomization (Semitones)").Out
    Min=UE.Subtract:Float(PrimaryOperand=^"Pitch (Semitones)", AdditionalOperands=^"Pitch Randomization (Semitones)").Out
    Next=VariableAccessor.Trigger(Variable=VariableMutator.Trigger(Variable=@n8.Variable, Value=@TriggerDelay_1.Out).Variable).Value
   )
   ArrayRandomGet_1 = Array.Random Get:WaveAsset:Array(In Array=^Wave Assets, Next=@n12.On Next) {also {defaults {Enable Shared State="<MetaSoundBuilderResult.SUCCEEDED: 0>", Weights []}}}
   n13 = VariableAccessor.Float(Variable=VariableMutator.Float(Variable=InitVariable.Float().Variable, Value=@n12.Value).Variable)
   WavePlayer_0 = UE.Wave Player:Mono(Play=@ArrayRandomGet_1.On Next, Wave Asset=@ArrayRandomGet_1.Value, Pitch Shift=VariableAccessor.Float(Variable=@n13.Variable).Value) {also {version [1,1]}}
   ADSREnvelopeADSREnvelope_0 = ADSR Envelope.ADSR Envelope:Float(
    Attack Curve=^Crossfade Curve
    Release Curve=^Crossfade Curve
    Release Time=UE.ConversionFloatToTime(In=^"Crossfade Time (Seconds)").Out
    Trigger Release=VariableAccessor.Trigger(Variable=VariableMutator.Trigger(Variable=@n6.Variable, Value=@WavePlayer_1.On Play).Variable).Value
    Attack Time=@n7.Value
    Trigger Attack=@WavePlayer_0.On Play
   ) {also {defaults {Hard Reset="<MetaSoundBuilderResult.SUCCEEDED: 0>", Sustain Level=1}}}
   AudioMixerAudioMixerMono2_0 = "AudioMixer.Audio Mixer (Mono, 2)"(
    Gain 0=UE.Multiply:Float(AdditionalOperands=@DecibelstoLinearGain_0.Linear Gain, PrimaryOperand=UE.Reroute(Value=@ADSREnvelopeADSREnvelope_1.Out Envelope).Value).Out
    Gain 1=UE.Multiply:Float(AdditionalOperands=@DecibelstoLinearGain_0.Linear Gain, PrimaryOperand=UE.Reroute(Value=@ADSREnvelopeADSREnvelope_0.Out Envelope).Value).Out
    In 0=UE.Reroute(Value=@WavePlayer_1.Out Mono).Value
    In 1=UE.Reroute(Value=@WavePlayer_0.Out Mono).Value
   )
   TriggerOnce_0 = UE.Trigger Once(Trigger In=@TriggerDelay_1.Out)
   TriggerRouteTriggerRouteTime2_0 = "TriggerRoute.Trigger Route (Time, 2):Time"(Value 0=UE.ConversionFloatToTime(In=^"Crossfade Time (Seconds)").Out, Set 0=@TriggerOnce_0.Trigger Out)
   VariableMutator.Time(Value=@TriggerRouteTriggerRouteTime2_0.Value, Variable=@n7.Variable)
   GetWaveDuration_1 = "UE.Get Wave Duration: "(Wave=@ArrayRandomGet_1.Value)
   n14 = UE.Divide:Float(
    PrimaryOperand=UE.ConversionTimeToFloat(In=@GetWaveDuration_1.Duration).Out
    AdditionalOperands=UE.Power:Float(AdditionalOperands=UE.Divide:Float(PrimaryOperand=@n13.Value) {also {defaults {AdditionalOperands=12}}}.Out) {also {defaults {PrimaryOperand=2}}}.Out
   )
   n15 = InitVariable.Float()
   n16 = UE.Subtract:Float(
    AdditionalOperands=^"Crossfade Time (Seconds)"
    PrimaryOperand=VariableAccessor.Float(Variable=VariableMutator.Float(Value=@n14.Out, Variable=@n15.Variable).Variable).Value
   )
   TriggerDelay_0 = UE.Trigger Delay(
    In=@ADSREnvelopeADSREnvelope_0.On Attack Triggered
    Delay Time=UE.ConversionFloatToTime(In=Clamp.Clamp:Float(In=@n16.Out) {also {defaults {Max=60, Min=0.01}}}.Value).Out
   )
   VariableMutator.Trigger(Value=@TriggerDelay_0.Out, Variable=@n3.Variable)
   VariableMutator.Trigger(Variable=@n2.Variable, Value=@WavePlayer_0.On Play)
  ]
  links [{dst="@UE.OutputFormat.Mono.Audio:0", src=AudioMixerAudioMixerMono2_0.Out}]
 }]
properties {
 assetImportData=&self:AssetImportData
 attenuationSettings=&/ATT_DragonCart_Beam
 bLooping=true
 duration=10000
 maxDistance=4500
 sampleRate=48000
 soundClassObject=&Fort_Ambient_Athena
 volume=4
}
