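# Looping mono MetaSound source for the DragonCart beam, crossfading endlessly between randomly picked beam waves with randomized pitch.
# A trigger sequence alternates two wave players, each start releasing the other's ADSR envelope.
# Pitch-corrected wave duration minus crossfade time delays a trigger that feeds back to pick the next wave.
# Attack time switches to the crossfade time after the first attack, so only later waves fade in.
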MetaSoundSource /DragonCartNarrative/Assets/Audio/Beam/MSS_DragonCart_Beam_V2_LP
inputs [
 Crossfade Curve {type=Float}
 "Crossfade Time (Seconds)" {type=Float}
 "Pitch (Semitones)" {type=Float}
 "Pitch Randomization (Semitones)" {type=Float}
 UE.Source.OnPlay {type=Trigger}
 "Volume (dB)" {type=Float}
 Wave Assets {type=WaveAsset:Array}
]
interfaces [UE.OutputFormat.Mono:1.0, UE.Source:1.1]
outputs ["UE.OutputFormat.Mono.Audio:0" {type=Audio}]
pages [Default {
  input_defaults {
   Crossfade Curve=1
   "Crossfade Time (Seconds)"=1
   "Pitch (Semitones)"=0
   "Pitch Randomization (Semitones)"=1
   UE.Source.OnPlay=null
   "Volume (dB)"=3
   Wave Assets [
    &/DragonCart_Beam_04
    &/DragonCart_Beam_03
    &/DragonCart_Beam_02
    &/DragonCart_Beam_01
    &/DragonCart_Beam_05
   ]
  }
  nodes [
   DecibelstoLinearGain_0 = UE.Decibels to Linear Gain:Float(Decibels=^"Volume (dB)")
   n1 = VariableDeferredAccessor.Time(Variable=InitVariable.Time().Variable)
   n2 = VariableDeferredAccessor.Trigger(Variable=InitVariable.Trigger().Variable)
   n3 = VariableDeferredAccessor.Trigger(Variable=InitVariable.Trigger().Variable)
   TriggerAnyTriggerAny2_0 = "TriggerAny.Trigger Any (2)"(In 0=^UE.Source.OnPlay, In 1=@n3.Value)
   ArrayRandomGet_0 = Array.Random Get:WaveAsset:Array(In Array=^Wave Assets, Next=@TriggerAnyTriggerAny2_0.Out) {also {defaults {Enable Shared State="<MetaSoundBuilderResult.SUCCEEDED: 0>"}}}
   n4 = UE.RandomFloat(
    Max=UE.Add:Float(PrimaryOperand=^"Pitch (Semitones)", AdditionalOperands=^"Pitch Randomization (Semitones)").Out
    Next=@ArrayRandomGet_0.On Next
    Min=UE.Subtract:Float(PrimaryOperand=^"Pitch (Semitones)", AdditionalOperands=^"Pitch Randomization (Semitones)").Out
   )
   TriggerSequence2_0 = "UE.Trigger Sequence (2)"(In=@n4.On Next)
   n5 = Value.Value:Float(Target Value=@n4.Value, Set=@TriggerSequence2_0.Out 0) {also {defaults {Init Value=0}}}
   n6 = VariableAccessor.Float(Variable=VariableMutator.Float(Variable=InitVariable.Float().Variable, Value=@n5.Output Value).Variable)
   n7 = VariableDeferredAccessor.Trigger(Variable=InitVariable.Trigger().Variable)
   n8 = VariableAccessor.Trigger(Variable=VariableMutator.Trigger(Value=@n5.On Set, Variable=@n7.Variable).Variable)
   WavePlayer_1 = UE.Wave Player:Mono(
    Wave Asset=@ArrayRandomGet_0.Value
    Pitch Shift=VariableAccessor.Float(Variable=@n6.Variable).Value
    Play=VariableAccessor.Trigger(Variable=@n8.Variable).Value
   ) {also {version [1,1]}}
   ADSREnvelopeADSREnvelope_1 = ADSR Envelope.ADSR Envelope:Float(
    Attack Curve=^Crossfade Curve
    Release Curve=^Crossfade Curve
    Release Time=UE.ConversionFloatToTime(In=^"Crossfade Time (Seconds)").Out
    Attack Time=@n1.Value
    Trigger Release=@n2.Value
    Trigger Attack=@WavePlayer_1.On Play
   ) {also {defaults {Hard Reset="<MetaSoundBuilderResult.SUCCEEDED: 0>", Sustain Level=1}}}
   n9 = VariableDeferredAccessor.Time(Variable=@n1.Variable)
   n10 = Value.Value:Float(Target Value=@n4.Value, Set=@TriggerSequence2_0.Out 1) {also {defaults {Init Value=0}}}
   n11 = VariableAccessor.Float(Variable=VariableMutator.Float(Variable=InitVariable.Float().Variable, Value=@n10.Output Value).Variable)
   n12 = VariableAccessor.Trigger(Variable=VariableMutator.Trigger(Value=@n10.On Set, Variable=@n2.Variable).Variable)
   WavePlayer_0 = UE.Wave Player:Mono(Wave Asset=@ArrayRandomGet_0.Value, Pitch Shift=@n11.Value, Play=VariableAccessor.Trigger(Variable=@n12.Variable).Value) {also {version [1,1]}}
   ADSREnvelopeADSREnvelope_0 = ADSR Envelope.ADSR Envelope:Float(
    Attack Curve=^Crossfade Curve
    Release Curve=^Crossfade Curve
    Release Time=UE.ConversionFloatToTime(In=^"Crossfade Time (Seconds)").Out
    Attack Time=@n9.Value
    Trigger Release=@n7.Value
    Trigger Attack=@WavePlayer_0.On Play
   ) {also {defaults {Hard Reset="<MetaSoundBuilderResult.SUCCEEDED: 0>", Sustain Level=1}}}
   AudioMixerAudioMixerMono2_0 = "AudioMixer.Audio Mixer (Mono, 2)"(
    Gain 0=UE.Multiply:Float(AdditionalOperands=@DecibelstoLinearGain_0.Linear Gain, PrimaryOperand=UE.Reroute(Value=@ADSREnvelopeADSREnvelope_1.Out Envelope).Value).Out
    Gain 1=UE.Multiply:Float(AdditionalOperands=@DecibelstoLinearGain_0.Linear Gain, PrimaryOperand=UE.Reroute(Value=@ADSREnvelopeADSREnvelope_0.Out Envelope).Value).Out
    In 0=UE.Reroute(Value=@WavePlayer_1.Out Mono).Value
    In 1=UE.Reroute(Value=@WavePlayer_0.Out Mono).Value
   )
   TriggerOnce_0 = UE.Trigger Once(Trigger In=@ADSREnvelopeADSREnvelope_1.On Attack Triggered)
   TriggerRouteTriggerRouteTime2_0 = "TriggerRoute.Trigger Route (Time, 2):Time"(Value 0=UE.ConversionFloatToTime(In=^"Crossfade Time (Seconds)").Out, Set 0=@TriggerOnce_0.Trigger Out)
   VariableMutator.Time(Value=@TriggerRouteTriggerRouteTime2_0.Value, Variable=@n9.Variable)
   GetWaveDuration_0 = "UE.Get Wave Duration: "(Wave=@ArrayRandomGet_0.Value)
   n13 = UE.Divide:Float(PrimaryOperand=VariableAccessor.Float(Variable=@n11.Variable).Value) {also {defaults {AdditionalOperands=12}}}
   n14 = UE.Divide:Float(PrimaryOperand=@n6.Value) {also {defaults {AdditionalOperands=12}}}
   TriggerRouteTriggerRouteFloat2_0 = "TriggerRoute.Trigger Route (Float, 2):Float"(
    Value 1=UE.Divide:Float(
     PrimaryOperand=UE.ConversionTimeToFloat(In=@GetWaveDuration_0.Duration).Out
     AdditionalOperands=UE.Power:Float(AdditionalOperands=@n13.Out) {also {defaults {PrimaryOperand=2}}}.Out
    ).Out
    Value 0=UE.Divide:Float(
     PrimaryOperand=UE.ConversionTimeToFloat(In=@GetWaveDuration_0.Duration).Out
     AdditionalOperands=UE.Power:Float(AdditionalOperands=@n14.Out) {also {defaults {PrimaryOperand=2}}}.Out
    ).Out
    Set 1=@n12.Value
    Set 0=@n8.Value
   )
   n15 = Clamp.Clamp:Float(In=UE.Subtract:Float(AdditionalOperands=^"Crossfade Time (Seconds)", PrimaryOperand=@TriggerRouteTriggerRouteFloat2_0.Value).Out) {also {defaults {Max=60, Min=0.01}}}
   VariableMutator.Trigger(
    Value=UE.Pipe(Delay Time=UE.ConversionFloatToTime(In=@n15.Value).Out, In=@TriggerRouteTriggerRouteFloat2_0.On Set).Out
    Variable=@n3.Variable
   )
  ]
  links [{dst="@UE.OutputFormat.Mono.Audio:0", src=AudioMixerAudioMixerMono2_0.Out}]
 }]
properties {
 assetImportData=&self:AssetImportData
 attenuationSettings=&/ATT_DragonCart_Beam
 bLooping=true
 duration=10000
 maxDistance=4500
 sampleRate=48000
 soundClassObject=&Fort_Ambient_Athena
}
