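# One-shot stereo MetaSound Source for the DragonCart beam lift start sound, apparently the first-person (1P) variant.
# On play, a random-get node picks one of three start wave variations and feeds it to a stereo Wave Player.
# The player's finish trigger ends the source, which uses the DragonCart beam attenuation and the player attacks and abilities sound class.
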MetaSoundSource /DragonCartNarrative/Assets/Audio/Beam/MSS_DragonCart_MatchMist_Beam_Lift_1P_Start
inputs [
 Sounds {type=WaveAsset:Array}
 UE.Source.OnPlay {type=Trigger}
]
interfaces [UE.OutputFormat.Stereo:1.0, UE.Source.OneShot:1.0, UE.Source:1.1]
outputs [
 "UE.OutputFormat.Stereo.Audio:0" {type=Audio}
 "UE.OutputFormat.Stereo.Audio:1" {type=Audio}
 UE.Source.OneShot.OnFinished {type=Trigger}
]
pages [Default {
  input_defaults {
   UE.Source.OnPlay=null
   Sounds [
    &/DragonCart_Beam_Lift_Start_03
    &/DragonCart_Beam_Lift_Start_02
    &/DragonCart_Beam_Lift_Start_01
   ]
  }
  nodes [
   ArrayRandomGet_0 = Array.Random Get:WaveAsset:Array(In Array=^Sounds, Next=^UE.Source.OnPlay)
   WavePlayer_0 = UE.Wave Player:Stereo(Play=@ArrayRandomGet_0.On Next, Wave Asset=@ArrayRandomGet_0.Value) {also {version [1,1]}}
  ]
  links [
   {dst="@UE.Source.OneShot.OnFinished", src=WavePlayer_0.On Finished}
   {dst="@UE.OutputFormat.Stereo.Audio:0", src=WavePlayer_0.Out Left}
   {dst="@UE.OutputFormat.Stereo.Audio:1", src=WavePlayer_0.Out Right}
  ]
 }]
properties {
 assetImportData=&self:AssetImportData
 attenuationSettings=&/ATT_DragonCart_Beam
 maxDistance=4500
 numChannels=2
 outputFormat=Stereo
 sampleRate=48000
 soundClassObject=/Game/Sounds/FortSoundClassesAndModes/Fort_06_Player_Attacks_Abilities
 volume=0.4
}
