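# Mono MetaSound source for the third-person looping lift sound of the DragonCart MatchMist beam.
# On play it starts a wave player on the beam lift asset and picks a random pitch target.
# An interpolator glides the pitch shift from zero up to that target over several seconds, giving each playback a varied rise.
# Beam attenuation and a concurrency limit govern it, under the player abilities sound class.
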MetaSoundSource /DragonCartNarrative/Assets/Audio/Beam/MSS_DragonCart_MatchMist_Beam_Lift_3P_LP
inputs [
 Sounds {type=WaveAsset}
 UE.Source.OnPlay {type=Trigger}
]
interfaces [UE.OutputFormat.Mono:1.0, UE.Source:1.1]
outputs ["UE.OutputFormat.Mono.Audio:0" {type=Audio}]
pages [Default {
  input_defaults {Sounds=&/DragonCart_Beam_Lift_LP, UE.Source.OnPlay=null}
  nodes [
   /MetaSoundWizard/MetaSounds/Loops/Single/MSS_Loop_Mono(Sound=^Sounds)
   n1 = UE.RandomFloat(Next=^UE.Source.OnPlay) {also {defaults {Max=6, Min=4}}}
   WavePlayer_0 = UE.Wave Player:Mono(
    Wave Asset=^Sounds
    Play=^UE.Source.OnPlay
    Pitch Shift=UE.InterpTo:Audio(Target=Value.Value:Float(Set=@n1.On Next, Target Value=@n1.Value) {also {defaults {Init Value=0}}}.Output Value) {also {defaults {Interp Time=6}}}.Value
   ) {also {defaults {Loop="<MetaSoundBuilderResult.SUCCEEDED: 0>"}, version [1,1]}}
  ]
  links [{dst="@UE.OutputFormat.Mono.Audio:0", src=WavePlayer_0.Out Mono}]
 }]
properties {
 assetImportData=&self:AssetImportData
 attenuationSettings=&/ATT_DragonCart_Beam
 bLooping=true
 bOverrideConcurrency=true
 duration=10000
 maxDistance=4500
 sampleRate=48000
 soundClassObject=/Game/Sounds/FortSoundClassesAndModes/Fort_06_Player_Attacks_Abilities
 volume=0.3
 concurrencyOverrides {maxCount=3}
}
